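# Sound cue for the looping rotation ambience of a white metal windmill spot at the WhiteWindmill POI.
# A Random node picks one of two equally weighted looping waves, which passes through a Modulator and an Enveloper that fades volume in.
# An Attenuation node reuses the Grassland outpost windmill stereo settings with overrides for falloff, spread, shape and listener focus.
# It plays in the ambient props sound class under prop concurrency rules.
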SoundCue /Game/Sounds/Helios/POI/WhiteWindmill/Spots/Spot_Windmill_Metal_White_Rotate_Loop_Cue
root=Attenuation_2
nodes [
 {
  id=Attenuation_2
  children [Enveloper_1]
  properties {
   attenuationSettings=/Game/Sounds/Artemis/Biomes/Grasslands/Grassland_Spots/Spot_Outpost_Windmill_Close_Stereo_Att
   bOverrideAttenuation=true
   attenuationOverrides {bEnableListenerFocus=true, falloffDistance=3800, stereoSpread=120, attenuationShapeExtents {x=750}}
  }
 }
 {
  id=Enveloper_1
  children [Modulator_1]
  properties {pitchCurve {editorCurveData {keys [{time=0}]}}, volumeCurve {editorCurveData {keys [<at 0 0>,<at 1 1>]}}}
 }
 {id=Modulator_1, children [Random_0], properties {volumeMax=1, volumeMin=1}}
 {id=Random_0, children [WavePlayer_2,WavePlayer_3], properties {weights [1,1]}}
 {
  id=WavePlayer_2
  properties {bLooping=true, soundWaveAssetPtr=&/Spot_Windmill_Metal_Rotate_Loop_02}
 }
 {
  id=WavePlayer_3
  properties {bLooping=true, soundWaveAssetPtr=&/Spot_Windmill_Metal_Rotate_Loop_01}
 }
]
properties {
 bEnableBusSends=false
 maxDistance=4550
 soundClassObject=&Fort_Ambient_Props_Athena
 volumeMultiplier=1.002
 concurrencyOverrides {maxCount=2}
 concurrencySet [/Game/Sounds/Concurrency_Rules/Athena_Prop]
 preEffectBusSends [{}]
}
